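# Material function that adds rim and surface color details to vehicle lenses by modifying the BaseColor of incoming material attributes.
# Two view-dependent masks come from scaled vertex normals dotted with the camera vector: a square-root falloff for rims and a high-power falloff for camera-facing surface glow.
# Each mask blends a squared color parameter, and both results are added to the original BaseColor, leaving other attributes untouched.
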MaterialFunction /VehicleCosmetics/SharedMaterials/MaterialFunctions/MF_Glass_LensEnhancement
graph {
 properties {description=Adds additional surface and rim color details to vehicle lenses, previewBlendMode=BLEND_Translucent}
 nodes [
  n1 = VectorParameter() {
   desc=Alpha increases the intensity of the color
   group=LensEnhancement
   parameterName=Rim_ColorXPower
   sortPriority=6
   defaultValue=<rgba 1 1 1 0.5>
  }
  n2 = ScalarParameter() {defaultValue=1.5, group=LensEnhancement, parameterName=RimNormalDistance, sortPriority=0}
  n3 = VertexNormalWS()
  n4 = CameraVectorWS()
  n5 = Power(Base=DotProduct(A=Multiply(A=@n2, B=@n3) {constB=1.5}, B=@n4)) {constExponent=0.5}
  n6 = Add(A=NamedRerouteUsage(), B=LinearInterpolate(A=Multiply(A=@n1.RGB, B=@n1.RGB), Alpha=Saturate(None=@n5)) {constB=0})
  n7 = VectorParameter() {
   desc=Alpha increases the intensity of the color
   group=LensEnhancement
   parameterName=Surface_ColorXPower
   sortPriority=8
   defaultValue=<rgba 1 1 1 0.5>
  }
  n8 = ScalarParameter() {defaultValue=0.45, group=LensEnhancement, parameterName=SurfaceNormalDistance, sortPriority=2}
  n9 = Multiply(A=DotProduct(A=@n4, B=Multiply(A=@n3, B=@n8) {constB=0.45})) {constB=2.25}
  n10 = ScalarParameter() {defaultValue=50, group=LensEnhancement, parameterName=SurfaceGlow_Power, sortPriority=4}
  n11 = LinearInterpolate(B=Multiply(A=@n7.RGB, B=@n7.RGB), Alpha=Saturate(None=Power(Base=@n9, Exp=@n10) {constExponent=50})) {constB=0.5}
  FunctionOutput(None=SetMaterialAttributes!(MaterialAttributes=NamedRerouteUsage(), BaseColor=Add(A=@n6, B=@n11)) {attributeSetTypes [#1]})
  n12 = GetMaterialAttributes(None=FunctionInput() {inputName=MA, inputType=FunctionInput_MaterialAttributes}) {attributeGetTypes [#1]}
  NamedRerouteDeclaration(None=@n12.BaseColor) {name=BaseColor, nodeColor=<rgba 1 0.5391 0 1>}
  NamedRerouteDeclaration(None=@n12.MaterialAttributes) {name=MA, nodeColor=<rgba 0.2188 0 1 1>}
 ]
}
