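# Material function for transparent glass in the shared vehicle cosmetics materials, previewed with a translucent blend mode.
# It is a thin wrapper around MF_Glass_Base, passing that function's Result straight to its single output.
# The base function exposes fresnel inner transparency, window tint amount, front color, metalness and roughness inputs, with no values wired to them here.
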MaterialFunction /VehicleCosmetics/SharedMaterials/MaterialFunctions/MF_Glass_Transparent
graph {
 properties {previewBlendMode=BLEND_Translucent}
 nodes [FunctionOutput(None=MaterialFunctionCall!(Fresnel Inner Transparency, Window Tint Amount, ColorFront, Metalness, Roughness) {materialFunction=&/MF_Glass_Base}.Result)]
}
