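# Translucent, lit glass material for the Ten vehicle body, with brake lighting and an optional infinity glow effect.
# A brake branch tints transparent glass through an emissive mask, adds opacity and emissive scaled by isBraking, then feeds the DeRez transition.
# A separate branch statically selects transparent (optionally lens-enhanced), dark-tint or opaque glass.
# Its InfinityOffset subgraph adds Fresnel and offset mask-layer emissive, with optional UV cropping, overlap masking and per-quality offsets.
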Material /VehicleCosmetics/Bodies/Ten/Material/M_Glass_Ten
graph {
 properties {
  bUseMaterialAttributes=true
  bUsedWithMeshDeformer=true
  bUsedWithSkeletalMesh=true
  blendMode=BLEND_Translucent
  pixelDepthOffsetMode=PDOM_Legacy
  translucencyLightingMode=TLM_SurfacePerPixelLighting
  usedShadingModels=Default Lit
 }
 nodes [
  n1 = MaterialFunctionCall() {materialFunction=/VehicleCosmetics/SharedMaterials/MaterialFunctions/MF_Glass_Transparent}
  n2 = StaticSwitchParameter(
   True=StaticSwitchParameter(
    True=MaterialFunctionCall!(MA=@n1.Result) {materialFunction=/VehicleCosmetics/SharedMaterials/MaterialFunctions/MF_Glass_LensEnhancement}.Result
    False=@n1.Result
   ) {group=LensEnhancement, parameterName=UseLensEnhancement, sortPriority=-1}
   False=StaticSwitchParameter(
    True=MaterialFunctionCall() {materialFunction=/VehicleCosmetics/SharedMaterials/MaterialFunctions/MF_Glass_DarkTint}.Result
    False=MaterialFunctionCall() {materialFunction=/VehicleCosmetics/SharedMaterials/MaterialFunctions/MF_Glass_Opaque}.Result
   ) {defaultValue=true, group=0_Glass_Types, parameterName="DarkTint<IO>Opaque", sortPriority=-5}
  ) {defaultValue=true, group=0_Glass_Types, parameterName=UseTransparent, sortPriority=-10}
  Composite!("in(MA)"=@n2) {subgraphName=InfinityOffset}
  n3 = VectorParameter() {group=Emissive, parameterName=EmissiveUVs_RG_UpperLeftCorner_BA_LowerRightCorner, defaultValue=<rgba 0 0 1 1>}
  n4 = AppendVector(A=@n3.RGB, B=@n3.RGB)
  n5 = TextureObjectParameter() {group=InfinityEffect, parameterName=EffectsMask, sortPriority=2, texture=/VehicleCosmetics/Textures/T_Black}
  Composite!(RGBA=@n4, T2D_In=@n5)
  Constant() {r=0.005}
  Constant() {r=0.33}
  Constant() {r=0.24}
  n6 = GetMaterialAttributes(None=MaterialFunctionCall() {materialFunction=&MF_Glass_Transparent}.Result) {attributeGetTypes [#1, #2, #3]}
  n7 = TextureSampleParameter2D() {
   group=Emissive
   parameterName=Emissive Masks
   samplerType=SAMPLERTYPE_LinearColor
   texture=/VehicleCosmetics/Bodies/Ten/Textures/T_Ten_Chassis_E
  }
  n8 = Power(Base=@n7.B, Exp=ScalarParameter() {defaultValue=1, group=Emissive, parameterName=MaskContrast})
  n9 = ScalarParameter() {defaultValue=1, group=Emissive, parameterName=isBraking}
  n10 = VectorParameter() {group=Emissive, parameterName=Brake_Color, defaultValue=<rgba 1 0 0 1>}
  n11 = SetMaterialAttributes!(
   MaterialAttributes=@n6.MaterialAttributes
   Base Color=LinearInterpolate(A=@n6.Base Color, B=VectorParameter() {group=Emissive, parameterName=Hex Color, defaultValue=<rgba 1 0 0 1>}.RGB, Alpha=@n7.B)
   Opacity=Add(A=@n6.Opacity, B=Multiply(A=@n8, B=@n9))
   Emissive Color=Multiply(A=Multiply(A=@n7.B, B=@n10.RGB), B=@n9)
   Normal=StaticSwitchParameter(True=NamedRerouteUsage(), False=@n6.Normal) {parameterName=Custom Normal?}
  ) {attributeSetTypes [
    #1
    #2
    #4
    #3
   ]}
  MaterialFunctionCall(In=@n11) {materialFunction=&MF_VehicleDeRez_Transition}
  MaterialExpressionReroute_4 = Reroute(None=@n2)
  n12 = GetMaterialAttributes(None=@MaterialExpressionReroute_4."in(MA)") {attributeGetTypes [#4]}
  NamedRerouteDeclaration(None=@n12.Emissive Color) {name=MAEmissive, nodeColor=<rgba 0.3828 1 0 1>}
  NamedRerouteDeclaration(None=@n12.MaterialAttributes) {name=MA, nodeColor=<rgba 0 0.0625 1 1>}
  n13 = TextureSampleParameter2D() {group=Emissive, parameterName=Custom Normal, samplerType=SAMPLERTYPE_Normal, texture=/VehicleCosmetics/Bodies/Ten/Textures/T_Ten_Chassis_N}
  NamedRerouteDeclaration(None=@n13.RGB) {name=Normal, nodeColor=<rgba 0.7109 1 0 1>}
  n14 = VectorParameter() {group=InfinityEffect, parameterName=CustomColor, sortPriority=4, defaultValue=<rgba 0.7766 0.3977 1 1>}
  n15 = Multiply(A=@n14.RGB, B=ScalarParameter() {defaultValue=6, group=InfinityEffect, parameterName=Brightness, sortPriority=6}) {constB=6}
  n16 = Multiply(A=Fresnel(Normal=VertexNormalWS()) {baseReflectFraction=0, exponent=14}, B=@n15)
  n17 = ScalarParameter() {defaultValue=0.33, group=InfinityEffect, parameterName=LayerOffset, sortPriority=12}
  MaterialExpressionReroute_1 = Reroute(None=@n4)
  MaterialExpressionReroute_9 = Reroute(None=@n5)
  n18 = StaticSwitchParameter(
   True=MaterialFunctionCall!(
    UVs
    TextureIn=@MaterialExpressionReroute_9.T2D_In
    UpperLeftCorner=ComponentMask(None=@MaterialExpressionReroute_1.RGBA) {g=true, r=true}
    LowerRightCorner=ComponentMask(None=@MaterialExpressionReroute_1.RGBA) {a=true, b=true}
    ExportFloat4
   ) {materialFunction=/Engine/Functions/Engine_MaterialFunctions01/Texturing/TextureCropping}.Crop UVs
   False=TextureSample(Tex=@MaterialExpressionReroute_9.T2D_In) {samplerType=SAMPLERTYPE_LinearColor, texture=/VehicleCosmetics/Textures/T_Black_Linear}.RGBA
  ) {defaultValue=true, desc=CROP_EMIS, group=Emissive, parameterName=CroppedEmissive}
  MaterialExpressionReroute_2 = Reroute(None=@n18)
  n19 = StaticSwitchParameter(True=ScalarParameter() {defaultValue=0.24, group=InfinityEffect, parameterName=LowQuality_LayerOffset, sortPriority=13}, False=@n17) {desc=Enables controls relating to shader quality settings, group=InfinityEffect, parameterName=EnableQualityControls, sortPriority=8}
  n20 = StaticSwitchParameter(True=ScalarParameter() {defaultValue=0.29, group=InfinityEffect, parameterName=MediumQuality_LayerOffset, sortPriority=14}, False=@n17) {desc=Enables controls relating to shader quality settings, group=InfinityEffect, parameterName=EnableQualityControls, sortPriority=8}
  n21 = MaterialFunctionCall!(
   Texture=@n5
   Offset=@n17
   RotateUVs
   GradientDarkness=ScalarParameter() {defaultValue=0.005, group=InfinityEffect, parameterName=GradientDarkness, sortPriority=16}
   GradientFalloff
   UVs=@MaterialExpressionReroute_2.Output
   Opt_LowQualityOffset=@n19
   Opt_MediumQualityOffset=@n20
  ) {materialFunction=/VehicleCosmetics/SharedMaterials/MaterialFunctions/MF_Infinity_OffsetFunction}
  n22 = TextureSample(UVs=@MaterialExpressionReroute_2.Output, Tex=@n5)
  n23 = Multiply(A=@n22.B) {constB=1.098}
  n24 = Saturate(None=Multiply(A=Subtract(A=@n22.B) {constB=0.098}) {constB=1.1087})
  n25 = Floor(None=MaterialFunctionCall(X=Saturate(None=@n23), Y=OneMinus(None=@n24)) {materialFunction=&MakeFloat2}.Result)
  n26 = ComponentMask(None=@n21.Result) {g=true, r=true}
  n27 = ScalarParameter() {
   defaultValue=1
   desc="Use if you have areas where your effect will overlap (e.g. the windshield has support pilars that result in the mask overlapping) or needs to be masked to only appear within a designated area."
   group=InfinityEffect
   parameterName=EnableOverlapMasking
  }
  n28 = MaterialFunctionCall(Float2=LinearInterpolate(A=ComponentMask(None=@n21.Result) {g=true}, B=Multiply(A=@n25, B=@n26), Alpha=@n27)) {materialFunction=&BreakOutFloat2Components}
  n29 = Add(A=@n28.R, B=@n28.R)
  n30 = LinearInterpolate(A=NamedRerouteUsage(), B=Multiply(A=@n15, B=@n29), Alpha=Saturate(None=@n29))
  n31 = Saturate(None=ScalarParameter() {defaultValue=1, group=InfinityEffect, parameterName=SelectionFloorFade, sliderMax=1, sortPriority=8})
  n32 = SetMaterialAttributes!(
   MaterialAttributes=NamedRerouteUsage()
   Emissive Color=Add(A=Multiply(A=@n16) {constB=0.33}, B=Multiply(A=@n30, B=@n31) {constB=0.4, desc=selectionfloorfade})
  ) {attributeSetTypes [#4]}
  PinBase!(Result=@n32) {reroutePins [Result {expression=&self:MaterialExpressionReroute_5}]}
  PinBase!() {reroutePins ["in(MA)" {expression=&self:MaterialExpressionReroute_4}]}
  PinBase() {reroutePins [
    RGBA {expression=&self:MaterialExpressionReroute_1}
    T2D_In {expression=&self:MaterialExpressionReroute_9}
   ]}
  PinBase!(Output=@n18) {reroutePins [Output {expression=&self:MaterialExpressionReroute_2}]}
  MaterialExpressionReroute_5 = Reroute(None=@n32)
  StaticSwitchParameter(True=@MaterialExpressionReroute_5.Result, False=@n2) {group=Effects, parameterName=UseInfinityOffsetFX}
 ]
}
