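# Opaque lit material for the license plate of the Ten vehicle body, usable on skeletal meshes and mesh deformers.
# A texture parameter defaulting to the plate texture supplies base color, while scalar parameters set uniform metallic and roughness.
# These are packed through Set Material Attributes and passed into the shared MF_Vehicle_FNPostFX function, following the material attributes workflow.
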Material /VehicleCosmetics/Bodies/Ten/Material/M_Ten_LicensePlate
graph {
 properties {
  bAutomaticallySetUsageInEditor=false
  bUseMaterialAttributes=true
  bUsedWithMeshDeformer=true
  bUsedWithSkeletalMesh=true
  usedShadingModels=Default Lit
 }
 nodes [MaterialFunctionCall!(In=SetMaterialAttributes!(
    MaterialAttributes
    Base Color=TextureSampleParameter2D() {parameterName=Base Color, texture=/VehicleCosmetics/Bodies/Ten/Textures/T_Ten_LicensePlate}.RGB
    Metallic=ScalarParameter() {parameterName=Metalic}
    Roughness=ScalarParameter() {defaultValue=0.5, parameterName=Roughness}
   ) {attributeSetTypes [#1, #2, #3]}) {materialFunction=/VehicleCosmetics/Materials/MaterialFunctions/MF_Vehicle_FNPostFX}]
}
