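# Physics asset for the Body_Ten vehicle cosmetic body mesh, with a BR suffix in its name.
# It holds a single body on the root bone with no collision shapes listed.
# It has no constraints, so no joints link bones for ragdoll or secondary motion.
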PhysicsAsset /VehicleCosmetics/Bodies/Ten/Mesh/Body_Ten_Physics_BR
bodies {root []}
constraints []
