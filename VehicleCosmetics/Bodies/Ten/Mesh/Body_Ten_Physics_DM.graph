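# Physics asset for the Ten vehicle body cosmetic, giving its mesh a simple collision shape.
# A single body on the root bone holds one box, roughly car-sized and offset slightly forward and raised above the bone origin.
# It has no other bodies or constraints, so it defines no per-part collision or jointed motion.
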PhysicsAsset /VehicleCosmetics/Bodies/Ten/Mesh/Body_Ten_Physics_DM
bodies {root [{shapeType=Box, extentX=444, extentY=216, extentZ=116, shapeName=root_box, center=<23 0 86.5>}]}
constraints []
