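# Material function for the Livery03 skin of vehicle body Ten, overriding BaseColor on the MF_Vehicle_Body_Base_v2 material attributes.
# Nested lerps blend scaled-down BaseColor and AccentColor parameters, offset by small counterbalance tints, into contrast-tuned rings.
# Ring masks come from biased, scaled dot products, with declared reroutes supplying tangent-space camera vector, world normal and tiled sparkle-flake noise.
# A final bias-scale and saturate boosts contrast.
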MaterialFunction /VehicleCosmetics/Bodies/Ten/Skins/MF_Skins_Ten_Livery03
graph {nodes [
  Comment() {text=Adjust for channel compression}
  n1 = Saturate(None=Add(A=Constant3Vector() {desc=Counterbalance Color, constant=<rgba 0.0395 0.0086 0.0242 1>}, B=NamedRerouteUsage()))
  n2 = Constant3Vector() {desc=Counterbalance Color, constant=<rgba 0.0437 0.0395 0.0723 1>}
  n3 = NamedRerouteUsage()
  n4 = Saturate(None=Add(A=Constant3Vector() {desc=Counterbalance Color, constant=<rgba 0.0296 0.0137 0 1>}, B=NamedRerouteUsage()))
  n5 = Constant3Vector() {desc=Counterbalance Color, constant=<rgba 0.0137 0.0242 0.0137 1>}
  n6 = NamedRerouteUsage()
  n7 = Constant3Vector() {desc=Counterbalance Color, constant=<rgba 0.0529 0.0437 0.0529 1>}
  n8 = DotProduct(A=Constant3Vector() {constant=<rgba 0 0 0.87 0>}, B=NamedRerouteUsage())
  n9 = LinearInterpolate(
   A=Saturate(None=Add(A=@n5, B=@n6))
   B=Saturate(None=Add(A=@n6, B=@n7))
   Alpha=Saturate(None=ConstantBiasScale(None=@n8) {bias=-0.859, desc="Bias == Size of Ring; Scale == Strength of contrast", scale=26})
  )
  n10 = DotProduct(A=Constant3Vector() {constant=<rgba 0 0 0.86 0>}, B=NamedRerouteUsage())
  n11 = NamedRerouteUsage()
  n12 = Saturate(None=Add(
    A=ConstantBiasScale(None=@n10) {bias=-0.725, desc="Bias == Size of Ring; Scale == Strength of contrast", scale=5}
    B=ConstantBiasScale(None=@n11) {bias=-0.01, scale=2}
   ))
  n13 = Constant3Vector() {constant=<rgba 0 0 0.95 0>}
  n14 = NamedRerouteUsage()
  n15 = Add(
   A=ConstantBiasScale(None=DotProduct(A=@n13, B=@n14)) {bias=-0.68, desc="Bias == Size of Ring; Scale == Strength of contrast", scale=7}
   B=ConstantBiasScale(None=NamedRerouteUsage()) {bias=-0.1, scale=3}
  )
  n16 = DotProduct(A=Constant3Vector() {constant=<rgba 0 0 0.97 0>}, B=NamedRerouteUsage())
  n17 = LinearInterpolate(
   A=Saturate(None=Add(A=@n2, B=@n3))
   B=LinearInterpolate(A=NamedRerouteUsage(), B=LinearInterpolate(A=@n4, B=@n9, Alpha=@n12), Alpha=Saturate(None=@n15))
   Alpha=Saturate(None=ConstantBiasScale(None=@n16) {bias=-0.68, desc="Bias == Size of Ring; Scale == Strength of contrast", scale=7})
  )
  n18 = DotProduct(A=Constant3Vector() {constant=<rgba 0 0 0.62 0>}, B=NamedRerouteUsage())
  n19 = NamedRerouteUsage()
  n20 = Saturate(None=Add(
    A=ConstantBiasScale(None=@n18) {bias=-0.2, desc="Bias == Size of Ring; Scale == Strength of contrast", scale=4}
    B=ConstantBiasScale(None=@n19) {bias=-0.01, scale=2}
   ))
  n21 = ConstantBiasScale(None=DotProduct(A=Constant3Vector() {constant=<rgba 0 0 0.68 0>}, B=NamedRerouteUsage())) {bias=-0.001, desc="Bias == Size of Ring; Scale == Strength of contrast", scale=6}
  n22 = ConstantBiasScale(None=LinearInterpolate(A=NamedRerouteUsage(), B=LinearInterpolate(A=@n1, B=@n17, Alpha=@n20), Alpha=Saturate(None=@n21))) {bias=0, desc="Lower Bias == Higher Contrast when \r\npaired with brighter Scale", scale=1.2}
  FunctionOutput(None=SetMaterialAttributes!(
    MaterialAttributes=MaterialFunctionCall() {materialFunction=/VehicleCosmetics/Materials/VehicleParents_Version2/Material_Functions/MF_Vehicle_Body_Base_v2}.Result
    BaseColor=Saturate(None=@n22)
   ) {attributeSetTypes [#1]})
  NamedRerouteDeclaration(None=TextureSample(UVs=TextureCoordinate() {uTiling=70, vTiling=70}) {texture=/VehicleCosmetics/SharedTextures/T_Noise_Sparkle_Flake}.RGB) {name=SparkleMask_Out, nodeColor=<rgba 0 1 0.2031 1>}
  NamedRerouteDeclaration(None=Transform(None=CameraVectorWS()) {transformSourceType=TRANSFORMSOURCE_World, transformType=TRANSFORM_Tangent}) {name=CameraVector_Out, nodeColor=<rgba 1 0 0.1172 1>}
  NamedRerouteDeclaration(None=VertexNormalWS()) {name=VNormal_Out, nodeColor=<rgba 0 1 0.4141 1>}
  n23 = VectorParameter() {group=Color, parameterName=BaseColor, sortPriority=1, defaultValue=<rgba 0.8148 0 0.0044 1>}
  NamedRerouteDeclaration(None=Saturate(None=Multiply(A=@n23.RGB) {constB=0.54})) {name=BaseColor_Out, nodeColor=<rgba 0 0.2 1 1>}
  n24 = VectorParameter() {group=Color, parameterName=AccentColor, sortPriority=2, defaultValue=<rgba 0 0.1779 1 1>}
  NamedRerouteDeclaration(None=Saturate(None=Multiply(A=@n24.RGB) {constB=0.54})) {name=AccentColor_Out, nodeColor=<rgba 1 1 1 1>}
 ]}
